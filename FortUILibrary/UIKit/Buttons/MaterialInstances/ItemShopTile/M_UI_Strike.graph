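# M_UI_Strike is a translucent UI material in the item shop tile button folder of FortUILibrary.
# It multiplies the Opacity from the shared modular-button state lerp function (idle, hovered, focused, pressed, disabled) by a box mask.
# MF_UI_Box builds that mask from normalized UI UVs as a full-size, square-cornered fill that keeps aspect ratio.
# The state inputs have nothing wired in, and the graph shows no link from the product to a material output.
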Material /FortUILibrary/UIKit/Buttons/MaterialInstances/ItemShopTile/M_UI_Strike
graph {
 properties {
  bUsedWithStaticMesh=false
  blendMode=BLEND_Translucent
  materialDomain=MD_UI
  pixelDepthOffsetMode=PDOM_Legacy
  usedShadingModels=Default Lit
 }
 nodes [Multiply(
   A=MaterialFunctionCall!(Idle, Idle Opacity, Hovered, Hovered Opacity, Focused, Focused Opacity, Pressed, Pressed Opacity, Disabled, Disabled Opacity) {materialFunction=/FortUILibrary/Systems/ModularButton/MaterialFunctions/MF_UIKit_ButtonStatesBase_Lerp}.Opacity
   B=MaterialFunctionCall(
    UVs=MaterialFunctionCall() {materialFunction=&GetUserInterfaceUV}.Normalized UV
    Size=Constant() {r=1}
    CornerRoundness=Constant()
    KeepAspectRatio=StaticBool() {value=true}
   ) {materialFunction=&MF_UI_Box}.Fill
  )]
}
